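; Light controller Blueprint that feeds its position, radius, color and intensity into the GrillMoon Material Parameter Collection, apparently for static cloud lighting.
; GetParamNames uses MPCIndex to pick one of four MatFX LocRad/Color parameter pairs.
; When bSendUpdates is set, UpdateMPC packs location/radius and color/intensity into two vectors and resizes the Sphere component to Radius.
; Tick and construction script are empty, and nothing here calls UpdateMPC.
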(event EventTick
  (params (DeltaSeconds :type Float%20%28single-precision%29)))

(fn GetParamNames
  (return (Utilities|Select :Index (Variables|Default|GetMPCIndex) :Option%200 "MatFX.01.LocRad" :Option%201 "MatFX.02.LocRad" :Option%202 "MatFX.03.LocRad" :Option%203 "MatFX.04.LocRad") (Utilities|Select :Index (Variables|Default|GetMPCIndex) :Option%200 "MatFX.01.Color" :Option%201 "MatFX.02.Color" :Option%202 "MatFX.03.Color" :Option%203 "MatFX.04.Color")))

(fn UpdateMPC
  (bind (_location_radius _color_intensity) (CallFunction|GetParamNames))
  (bind (_color_r _color_g _color_b) (Math|Color|BreakColor (Variables|GetColor_R)))
  (if (Variables|Default|GetbSendUpdates)
    (Rendering|Material|SetVectorParameterValue "/GrillMoon_Countdown/Assets/FX/MPC_GrillMoon.MPC_GrillMoon" _location_radius (Utilities|Struct|MakeLinearColor (.x (Transformation|GetActorLocation)) (.y (Transformation|GetActorLocation)) (.z (Transformation|GetActorLocation)) (Variables|Default|GetRadius)) :!class /Script/Engine.KismetMaterialLibrary)
    (Rendering|Material|SetVectorParameterValue "/GrillMoon_Countdown/Assets/FX/MPC_GrillMoon.MPC_GrillMoon" _color_intensity (Utilities|Struct|MakeLinearColor _color_r _color_g _color_b (Variables|Default|GetIntensity)) :!class /Script/Engine.KismetMaterialLibrary)
    (Components|Sphere|SetSphereRadius (Variables|Default|GetSphere) (Variables|Default|GetRadius) false)))

(fn ConstructionScript)
