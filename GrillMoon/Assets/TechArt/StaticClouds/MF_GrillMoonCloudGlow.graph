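# Material function that adds a localized colored glow to Grill Moon's static clouds, driven by location/radius and color vector inputs.
# A sphere mask around the location, with hardness and opacity from MPC_GrillMoon, samples world position warped by the Alpha input to break up its edge.
# Inside the mask, a dimmed base color gains the glow color, scaled by noise-modulated intensity and a time-of-day multiplier.
# A saturated vector alpha gates the final blend.
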MaterialFunction /GrillMoon/Assets/TechArt/StaticClouds/MF_GrillMoonCloudGlow
graph {nodes [
  n1 = FunctionInput() {inputName=Alpha, inputType=FunctionInput_Scalar, sortPriority=2}
  n2 = Add(
   A=Multiply(A=Subtract(A=@n1) {constB=0.5}, B=ScalarParameter() {defaultValue=2, group=Grill Moon, parameterName=CloudNoiseMaskWarp})
   B=WorldPosition().XYZ
  )
  n3 = NamedRerouteUsage()
  n4 = ComponentMask(None=@n3) {b=true, g=true, r=true}
  n5 = ComponentMask(None=@n3) {a=true}
  n6 = CollectionParameter() {collection=/GrillMoon_Countdown/Assets/FX/MPC_GrillMoon, parameterName=Clouds_Glow_Hardness}
  n7 = Saturate(None=Multiply(
    A=SphereMask(A=@n2, B=@n4, Radius=@n5, Hardness=@n6) {hardnessPercent=50}
    B=CollectionParameter() {collection=&MPC_GrillMoon, parameterName=Clouds_Glow_Opacity}
   ))
  n8 = FunctionInput() {inputName=BaseColor, sortPriority=1}
  n9 = NamedRerouteUsage()
  n10 = ComponentMask(None=@n9) {a=true}
  n11 = LinearInterpolate(A=Multiply(A=@n10, B=ScalarParameter() {defaultValue=0.5, group=Grill Moon, parameterName=CloudColorNoiseMultiplier}), B=@n10, Alpha=@n1)
  n12 = FunctionInput() {bUsePreviewValueAsDefault=true, inputName=TimeOfDayMulti, inputType=FunctionInput_Scalar, sortPriority=3, previewValue=<1 0 0 1>}
  n13 = Add(A=Divide(A=@n8), B=Multiply(A=ComponentMask(None=@n9) {b=true, g=true, r=true}, B=Multiply(A=@n11, B=@n12)))
  n14 = ComponentMask(None=NamedRerouteUsage()) {a=true}
  FunctionOutput(None=LinearInterpolate(A=@n8, B=LinearInterpolate(A=@n8, B=@n13, Alpha=@n7), Alpha=Saturate(None=@n14)))
  NamedRerouteDeclaration(None=FunctionInput() {inputName=MPC Color, inputType=FunctionInput_Vector4, sortPriority=7, previewValue=<1 0 0 1>}) {name=Input.MPC.Color, nodeColor=<rgba 1 0.9844 0 1>}
  n15 = FunctionInput() {inputName=MPC Loc Rad, inputType=FunctionInput_Vector4, sortPriority=6, previewValue=<1 0 0 1>}
  NamedRerouteDeclaration(None=@n15) {name=Input.MPC.Loc, nodeColor=<rgba 0.6641 0 1 1>}
 ]}
